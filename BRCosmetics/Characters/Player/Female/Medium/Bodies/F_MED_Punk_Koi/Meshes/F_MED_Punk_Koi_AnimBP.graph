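# Animation Blueprint for the F_MED_Punk_Koi cosmetic body that adds procedural secondary motion to the sleeves and skirt.
# Sleeves use spring-driven Anim Dynamics.
# From cached pose PreDressRBAN, a bool blend picks between back-skirt constraints with Anim Dynamics and a pelvis-space RigidBody simulation with overridden gravity.
# Chained constraints then make front panels follow thigh bones and average in-between panels from their neighbours, before further Anim Dynamics add sway.
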AnimBlueprint /BRCosmetics/Characters/Player/Female/Medium/Bodies/F_MED_Punk_Koi/Meshes/F_MED_Punk_Koi_AnimBP
target_skeleton=&/F_MED_Punk_Koi_Skeleton
graphs [
 AnimGraph {nodes [
   n1 = Anim Dynamics - Bone: dyn_sleeve_r_inner(Alpha=1, ExternalForce="0,0,0") {
    angularDampingOverride=0.999
    angularSpringConstant=5000
    bAngularSpring=true
    bGravityOverrideInSimSpace=true
    bOverrideAngularDamping=true
    gravityScale=0.35
    simulationSpace=RootRelative
    alphaBoolBlend {blendOption=Linear}
    alphaScaleBiasClamp {scale=0.2}
    boundBone {boneName=dyn_sleeve_r_inner}
    componentAppliedLinearAccClamp=<6000 6000 6000>
    componentLinearAccScale=<0.3 0.3 0.7>
    componentLinearVelScale=<0.3 0.3 0.7>
    physicsBodyDefinitions [{
      sphereCollisionRadius=0
      boundBone {boneName=dyn_sleeve_r_inner}
      boxExtents=<5 15 15>
      constraintSetup {angularLimitsMax=<60 20 30>, angularLimitsMin=<-60 -30 -80>, angularTarget=<1 0 0>}
      localJointOffset=<5 0 0>
     }]
   }
   n2 = Anim Dynamics - Bone: dyn_sleeve_r_lower(ComponentPose=@n1, Alpha=1, ExternalForce="0,0,0") {
    angularDampingOverride=0.999
    angularSpringConstant=5000
    bAngularSpring=true
    bGravityOverrideInSimSpace=true
    bOverrideAngularDamping=true
    gravityScale=0.35
    simulationSpace=RootRelative
    alphaBoolBlend {blendOption=Linear}
    alphaScaleBiasClamp {scale=0.2}
    boundBone {boneName=dyn_sleeve_r_lower}
    componentAppliedLinearAccClamp=<6000 6000 6000>
    componentLinearAccScale=<0.3 0.3 0.7>
    componentLinearVelScale=<0.3 0.3 0.7>
    physicsBodyDefinitions [{
      sphereCollisionRadius=0
      boundBone {boneName=dyn_sleeve_r_lower}
      boxExtents=<5 15 15>
      constraintSetup {angularLimitsMax=<60 20 30>, angularLimitsMin=<-60 -30 -80>, angularTarget=<1 0 0>}
      localJointOffset=<5 0 0>
     }]
   }
   n3 = Anim Dynamics - Bone: dyn_sleeve_r_outter(ComponentPose=@n2, Alpha=1, ExternalForce="0,0,0") {
    angularDampingOverride=0.999
    angularSpringConstant=5000
    bAngularSpring=true
    bGravityOverrideInSimSpace=true
    bOverrideAngularDamping=true
    gravityScale=0.35
    simulationSpace=RootRelative
    alphaBoolBlend {blendOption=Linear}
    alphaScaleBiasClamp {scale=0.2}
    boundBone {boneName=dyn_sleeve_r_outter}
    componentAppliedLinearAccClamp=<6000 6000 6000>
    componentLinearAccScale=<0.3 0.3 0.7>
    componentLinearVelScale=<0.3 0.3 0.7>
    physicsBodyDefinitions [{
      sphereCollisionRadius=0
      boundBone {boneName=dyn_sleeve_r_outter}
      boxExtents=<5 15 15>
      constraintSetup {angularLimitsMax=<60 40 30>, angularLimitsMin=<-60 -30 -80>, angularTarget=<1 0 0>}
      localJointOffset=<5 0 0>
     }]
   }
   n4 = Anim Dynamics - Bone: dyn_sleeve_l_inner(ComponentPose=@n3, Alpha=1, ExternalForce="0,0,0") {
    angularDampingOverride=0.999
    angularSpringConstant=5000
    bAngularSpring=true
    bGravityOverrideInSimSpace=true
    bOverrideAngularDamping=true
    gravityScale=0.35
    simulationSpace=RootRelative
    alphaBoolBlend {blendOption=Linear}
    alphaScaleBiasClamp {scale=0.2}
    boundBone {boneName=dyn_sleeve_l_inner}
    componentAppliedLinearAccClamp=<6000 6000 6000>
    componentLinearAccScale=<0.3 0.3 0.7>
    componentLinearVelScale=<0.3 0.3 0.7>
    physicsBodyDefinitions [{
      sphereCollisionRadius=0
      boundBone {boneName=dyn_sleeve_l_inner}
      boxExtents=<5 15 15>
      constraintSetup {angularLimitsMax=<60 20 30>, angularLimitsMin=<-60 -30 -80>, angularTarget=<1 0 0>}
      localJointOffset=<5 0 0>
     }]
   }
   n5 = Anim Dynamics - Bone: dyn_sleeve_l_lower(ComponentPose=@n4, Alpha=1, ExternalForce="0,0,0") {
    angularDampingOverride=0.999
    angularSpringConstant=5000
    bAngularSpring=true
    bGravityOverrideInSimSpace=true
    bOverrideAngularDamping=true
    gravityScale=0.35
    simulationSpace=RootRelative
    alphaBoolBlend {blendOption=Linear}
    alphaScaleBiasClamp {scale=0.2}
    boundBone {boneName=dyn_sleeve_l_lower}
    componentAppliedLinearAccClamp=<6000 6000 6000>
    componentLinearAccScale=<0.3 0.3 0.7>
    componentLinearVelScale=<0.3 0.3 0.7>
    physicsBodyDefinitions [{
      sphereCollisionRadius=0
      boundBone {boneName=dyn_sleeve_l_lower}
      boxExtents=<5 15 15>
      constraintSetup {angularLimitsMax=<60 20 30>, angularLimitsMin=<-60 -30 -80>, angularTarget=<1 0 0>}
      localJointOffset=<5 0 0>
     }]
   }
   Anim Dynamics - Bone: dyn_sleeve_l_outter(ComponentPose=@n5, Alpha=1, ExternalForce="0,0,0") {
    angularDampingOverride=0.999
    angularSpringConstant=5000
    bAngularSpring=true
    bGravityOverrideInSimSpace=true
    bOverrideAngularDamping=true
    gravityScale=0.35
    simulationSpace=RootRelative
    alphaBoolBlend {blendOption=Linear}
    alphaScaleBiasClamp {scale=0.2}
    boundBone {boneName=dyn_sleeve_l_outter}
    componentAppliedLinearAccClamp=<6000 6000 6000>
    componentLinearAccScale=<0.3 0.3 0.7>
    componentLinearVelScale=<0.3 0.3 0.7>
    physicsBodyDefinitions [{
      sphereCollisionRadius=0
      boundBone {boneName=dyn_sleeve_l_outter}
      boxExtents=<5 15 15>
      constraintSetup {angularLimitsMax=<60 20 30>, angularLimitsMin=<-60 -30 -80>, angularTarget=<1 0 0>}
      localJointOffset=<5 0 0>
     }]
   }
   n6 = "Constraint - dyn_skirt_bk_01 (RPP)"!(
    ComponentPose=Local To Component(LocalPose=Use cached pose 'PreDressRBAN'())
    Alpha=1
    ConstraintWeights_0=0
    ConstraintWeights_1:float=0.5
    ConstraintWeights_2:float=0.5
   ) -> (Pose) {
    alphaBoolBlend {blendOption=Linear}
    boneToModify {boneName=dyn_skirt_bk_01}
    constraintSetup [{transformType=Rotation, targetBone {boneName=pelvis}}, {targetBone {boneName=dyn_skirt_bk_01_l}}, {targetBone {boneName=dyn_skirt_bk_01_r}}]
    constraintWeights [1,1,1]
   }
   n7 = "Constraint - dyn_skirt_bk_02 (RPP)"!(ComponentPose=@n6, Alpha=1, ConstraintWeights_0=0, ConstraintWeights_1:float=0.5, ConstraintWeights_2:float=0.5) -> (Pose) {
    alphaBoolBlend {blendOption=Linear}
    boneToModify {boneName=dyn_skirt_bk_02}
    constraintSetup [{transformType=Rotation, targetBone {boneName=pelvis}}, {targetBone {boneName=dyn_skirt_bk_02_l}}, {targetBone {boneName=dyn_skirt_bk_02_r}}]
    constraintWeights [1,1,1]
   }
   n8 = "Constraint - dyn_skirt_bk_03 (RPP)"!(ComponentPose=@n7, Alpha=1, ConstraintWeights_0=0, ConstraintWeights_1:float=0.5, ConstraintWeights_2:float=0.5) -> (Pose) {
    alphaBoolBlend {blendOption=Linear}
    boneToModify {boneName=dyn_skirt_bk_03}
    constraintSetup [{transformType=Rotation, targetBone {boneName=pelvis}}, {targetBone {boneName=dyn_skirt_bk_03_l}}, {targetBone {boneName=dyn_skirt_bk_03_r}}]
    constraintWeights [1,1,1]
   }
   n9 = "Constraint - dyn_skirt_bk2_01_l (P)"(ComponentPose=@n8, Alpha=1, ConstraintWeights_0=1) {
    alphaBoolBlend {blendOption=Linear}
    boneToModify {boneName=dyn_skirt_bk2_01_l}
    constraintSetup [{targetBone {boneName=pelvis}}]
    constraintWeights [1]
   }
   n10 = "Constraint - dyn_skirt_bk2_02_l (P)"(ComponentPose=@n9, Alpha=0.3, ConstraintWeights_0=1) {
    alphaBoolBlend {blendOption=Linear}
    boneToModify {boneName=dyn_skirt_bk2_02_l}
    constraintSetup [{targetBone {boneName=pelvis}}]
    constraintWeights [1]
   }
   n11 = "Constraint - dyn_skirt_bk2_03_l (P)"(ComponentPose=@n10, Alpha=1, ConstraintWeights_0=1) {
    alphaBoolBlend {blendOption=Linear}
    boneToModify {boneName=dyn_skirt_bk2_03_l}
    constraintSetup [{targetBone {boneName=pelvis}}]
    constraintWeights [1]
   }
   n12 = "Constraint - dyn_skirt_bk_01_l (RRR)"!(ComponentPose=@n11, Alpha=1, ConstraintWeights_0=0, ConstraintWeights_1:float=0.5, ConstraintWeights_2:float=0.5) -> (Pose) {
    alphaBoolBlend {blendOption=Linear}
    boneToModify {boneName=dyn_skirt_bk_01_l}
    constraintSetup [
     {transformType=Rotation, targetBone {boneName=pelvis}}
     {transformType=Rotation, targetBone {boneName=dyn_skirt_bk_01}}
     {transformType=Rotation, targetBone {boneName=dyn_skirt_bk2_01_l}}
    ]
    constraintWeights [1,1,1]
   }
   n13 = "Constraint - dyn_skirt_bk_02_l (RRR)"!(ComponentPose=@n12, Alpha=1, ConstraintWeights_0=0, ConstraintWeights_1:float=0.5, ConstraintWeights_2:float=0.5) -> (Pose) {
    alphaBoolBlend {blendOption=Linear}
    boneToModify {boneName=dyn_skirt_bk_02_l}
    constraintSetup [
     {transformType=Rotation, targetBone {boneName=pelvis}}
     {transformType=Rotation, targetBone {boneName=dyn_skirt_bk_02}}
     {transformType=Rotation, targetBone {boneName=dyn_skirt_bk2_02_l}}
    ]
    constraintWeights [1,1,1]
   }
   n14 = "Constraint - dyn_skirt_bk_03_l (RRR)"!(ComponentPose=@n13, Alpha=1, ConstraintWeights_0=0, ConstraintWeights_1:float=0.5, ConstraintWeights_2:float=0.5) -> (Pose) {
    alphaBoolBlend {blendOption=Linear}
    boneToModify {boneName=dyn_skirt_bk_03_l}
    constraintSetup [
     {transformType=Rotation, targetBone {boneName=pelvis}}
     {transformType=Rotation, targetBone {boneName=dyn_skirt_bk_03}}
     {transformType=Rotation, targetBone {boneName=dyn_skirt_bk2_03_l}}
    ]
    constraintWeights [1,1,1]
   }
   n15 = "Constraint - dyn_skirt_bk_01_r (RRR)"!(ComponentPose=@n14, Alpha=1, ConstraintWeights_0=0, ConstraintWeights_1:float=0.5, ConstraintWeights_2:float=0.5) -> (Pose) {
    alphaBoolBlend {blendOption=Linear}
    boneToModify {boneName=dyn_skirt_bk_01_r}
    constraintSetup [
     {transformType=Rotation, targetBone {boneName=pelvis}}
     {transformType=Rotation, targetBone {boneName=dyn_skirt_bk_01}}
     {transformType=Rotation, targetBone {boneName=dyn_skirt_bk2_01_r}}
    ]
    constraintWeights [1,1,1]
   }
   n16 = "Constraint - dyn_skirt_bk_02_r (RRR)"!(ComponentPose=@n15, Alpha=1, ConstraintWeights_0=0, ConstraintWeights_1:float=0.5, ConstraintWeights_2:float=0.5) -> (Pose) {
    alphaBoolBlend {blendOption=Linear}
    boneToModify {boneName=dyn_skirt_bk_02_r}
    constraintSetup [
     {transformType=Rotation, targetBone {boneName=pelvis}}
     {transformType=Rotation, targetBone {boneName=dyn_skirt_bk_02}}
     {transformType=Rotation, targetBone {boneName=dyn_skirt_bk2_02_r}}
    ]
    constraintWeights [1,1,1]
   }
   n17 = "Constraint - dyn_skirt_bk_03_r (RRR)"!(ComponentPose=@n16, Alpha=1, ConstraintWeights_0=0, ConstraintWeights_1:float=0.5, ConstraintWeights_2:float=0.5) -> (Pose) {
    alphaBoolBlend {blendOption=Linear}
    boneToModify {boneName=dyn_skirt_bk_03_r}
    constraintSetup [
     {transformType=Rotation, targetBone {boneName=pelvis}}
     {transformType=Rotation, targetBone {boneName=dyn_skirt_bk_03}}
     {transformType=Rotation, targetBone {boneName=dyn_skirt_bk2_03_r}}
    ]
    constraintWeights [1,1,1]
   }
   n18 = "Constraint - dyn_skirt_bk2_01_r (P)"(ComponentPose=@n17, Alpha=1, ConstraintWeights_0=1) {
    alphaBoolBlend {blendOption=Linear}
    boneToModify {boneName=dyn_skirt_bk2_01_r}
    constraintSetup [{targetBone {boneName=pelvis}}]
    constraintWeights [1]
   }
   n19 = "Constraint - dyn_skirt_bk2_02_r (P)"(ComponentPose=@n18, Alpha=1, ConstraintWeights_0=1) {
    alphaBoolBlend {blendOption=Linear}
    boneToModify {boneName=dyn_skirt_bk2_02_r}
    constraintSetup [{targetBone {boneName=pelvis}}]
    constraintWeights [1]
   }
   n20 = "Constraint - dyn_skirt_bk2_03_r (P)"(ComponentPose=@n19, Alpha=1, ConstraintWeights_0=1) {
    alphaBoolBlend {blendOption=Linear}
    boneToModify {boneName=dyn_skirt_bk2_03_r}
    constraintSetup [{targetBone {boneName=pelvis}}]
    constraintWeights [1]
   }
   n21 = Anim Dynamics - Bone: dyn_skirt_bk_01_l(ComponentPose=@n20, Alpha=0.5, ExternalForce="0,0,0") {
    angularDampingOverride=0.9999
    angularSpringConstant=75
    bAngularSpring=true
    bGravityOverrideInSimSpace=true
    bOverrideAngularDamping=true
    simulationSpace=RootRelative
    alphaBoolBlend {blendOption=Linear}
    boundBone {boneName=dyn_skirt_bk_01_l}
    componentLinearAccScale=<1.2 1.2 1.2>
    componentLinearVelScale=<1.2 1.2 1.2>
    physicsBodyDefinitions [{
      sphereCollisionRadius=0
      boundBone {boneName=dyn_skirt_bk_01_l}
      boxExtents=<20 20 10>
      constraintSetup {angularLimitsMax=<25 0 25>, angularLimitsMin=<-25 -25 -25>, angularTarget=<0 0 0>}
      localJointOffset=<2 0 0>
     }]
   }
   n22 = Anim Dynamics - Bone: dyn_skirt_bk_02_l(ComponentPose=@n21, Alpha=0.5, ExternalForce="0,0,0") {
    angularDampingOverride=0.9999
    angularSpringConstant=75
    bAngularSpring=true
    bGravityOverrideInSimSpace=true
    bOverrideAngularDamping=true
    simulationSpace=RootRelative
    alphaBoolBlend {blendOption=Linear}
    boundBone {boneName=dyn_skirt_bk_02_l}
    componentLinearAccScale=<1.2 1.2 1.2>
    componentLinearVelScale=<1.2 1.2 1.2>
    physicsBodyDefinitions [{
      sphereCollisionRadius=0
      boundBone {boneName=dyn_skirt_bk_02_l}
      boxExtents=<20 20 10>
      constraintSetup {angularLimitsMax=<25 0 25>, angularLimitsMin=<-25 -25 -25>, angularTarget=<0 0 0>}
      localJointOffset=<2 0 0>
     }]
   }
   n23 = Anim Dynamics - Bone: dyn_skirt_bk_03_l(ComponentPose=@n22, Alpha=0.5, ExternalForce="0,0,0") {
    angularDampingOverride=0.9999
    angularSpringConstant=75
    bAngularSpring=true
    bGravityOverrideInSimSpace=true
    bOverrideAngularDamping=true
    simulationSpace=RootRelative
    alphaBoolBlend {blendOption=Linear}
    boundBone {boneName=dyn_skirt_bk_03_l}
    componentLinearAccScale=<1.2 1.2 1.2>
    componentLinearVelScale=<1.2 1.2 1.2>
    physicsBodyDefinitions [{
      sphereCollisionRadius=0
      boundBone {boneName=dyn_skirt_bk_03_l}
      boxExtents=<20 20 10>
      constraintSetup {angularLimitsMax=<25 0 25>, angularLimitsMin=<-25 -25 -25>, angularTarget=<0 0 0>}
      localJointOffset=<2 0 0>
     }]
   }
   n24 = Anim Dynamics - Bone: dyn_skirt_bk_01(ComponentPose=@n23, Alpha=0.5, ExternalForce="0,0,0") {
    angularDampingOverride=0.9999
    angularSpringConstant=75
    bAngularSpring=true
    bGravityOverrideInSimSpace=true
    bOverrideAngularDamping=true
    simulationSpace=RootRelative
    alphaBoolBlend {blendOption=Linear}
    boundBone {boneName=dyn_skirt_bk_01}
    componentLinearAccScale=<1.2 1.2 1.2>
    componentLinearVelScale=<1.2 1.2 1.2>
    physicsBodyDefinitions [{
      sphereCollisionRadius=0
      boundBone {boneName=dyn_skirt_bk_01}
      boxExtents=<20 20 10>
      constraintSetup {angularLimitsMax=<25 0 25>, angularLimitsMin=<-25 -25 -25>, angularTarget=<0 0 0>}
      localJointOffset=<2 0 0>
     }]
   }
   n25 = Anim Dynamics - Bone: dyn_skirt_bk_02(ComponentPose=@n24, Alpha=0.5, ExternalForce="0,0,0") {
    angularDampingOverride=0.9999
    angularSpringConstant=75
    bAngularSpring=true
    bGravityOverrideInSimSpace=true
    bOverrideAngularDamping=true
    simulationSpace=RootRelative
    alphaBoolBlend {blendOption=Linear}
    boundBone {boneName=dyn_skirt_bk_02}
    componentLinearAccScale=<1.2 1.2 1.2>
    componentLinearVelScale=<1.2 1.2 1.2>
    physicsBodyDefinitions [{
      sphereCollisionRadius=0
      boundBone {boneName=dyn_skirt_bk_02}
      boxExtents=<20 20 10>
      constraintSetup {angularLimitsMax=<25 0 25>, angularLimitsMin=<-25 -25 -25>, angularTarget=<0 0 0>}
      localJointOffset=<2 0 0>
     }]
   }
   n26 = Anim Dynamics - Bone: dyn_skirt_bk_03(ComponentPose=@n25, Alpha=0.5, ExternalForce="0,0,0") {
    angularDampingOverride=0.9999
    angularSpringConstant=75
    bAngularSpring=true
    bGravityOverrideInSimSpace=true
    bOverrideAngularDamping=true
    simulationSpace=RootRelative
    alphaBoolBlend {blendOption=Linear}
    boundBone {boneName=dyn_skirt_bk_03}
    componentLinearAccScale=<1.2 1.2 1.2>
    componentLinearVelScale=<1.2 1.2 1.2>
    physicsBodyDefinitions [{
      sphereCollisionRadius=0
      boundBone {boneName=dyn_skirt_bk_03}
      boxExtents=<20 20 10>
      constraintSetup {angularLimitsMax=<25 0 25>, angularLimitsMin=<-25 -25 -25>, angularTarget=<0 0 0>}
      localJointOffset=<2 0 0>
     }]
   }
   n27 = Anim Dynamics - Bone: dyn_skirt_bk_01_r(ComponentPose=@n26, Alpha=0.5, ExternalForce="0,0,0") {
    angularDampingOverride=0.9999
    angularSpringConstant=75
    bAngularSpring=true
    bGravityOverrideInSimSpace=true
    bOverrideAngularDamping=true
    simulationSpace=RootRelative
    alphaBoolBlend {blendOption=Linear}
    boundBone {boneName=dyn_skirt_bk_01_r}
    componentLinearAccScale=<1.2 1.2 1.2>
    componentLinearVelScale=<1.2 1.2 1.2>
    physicsBodyDefinitions [{
      sphereCollisionRadius=0
      boundBone {boneName=dyn_skirt_bk_01_r}
      boxExtents=<20 20 10>
      constraintSetup {angularLimitsMax=<25 0 25>, angularLimitsMin=<-25 -25 -25>, angularTarget=<0 0 0>}
      localJointOffset=<2 0 0>
     }]
   }
   n28 = Anim Dynamics - Bone: dyn_skirt_bk_02_r(ComponentPose=@n27, Alpha=0.5, ExternalForce="0,0,0") {
    angularDampingOverride=0.9999
    angularSpringConstant=75
    bAngularSpring=true
    bGravityOverrideInSimSpace=true
    bOverrideAngularDamping=true
    simulationSpace=RootRelative
    alphaBoolBlend {blendOption=Linear}
    boundBone {boneName=dyn_skirt_bk_02_r}
    componentLinearAccScale=<1.2 1.2 1.2>
    componentLinearVelScale=<1.2 1.2 1.2>
    physicsBodyDefinitions [{
      sphereCollisionRadius=0
      boundBone {boneName=dyn_skirt_bk_02_r}
      boxExtents=<20 20 10>
      constraintSetup {angularLimitsMax=<25 0 25>, angularLimitsMin=<-25 -25 -25>, angularTarget=<0 0 0>}
      localJointOffset=<2 0 0>
     }]
   }
   n29 = Anim Dynamics - Bone: dyn_skirt_bk_03_r(ComponentPose=@n28, Alpha=0.5, ExternalForce="0,0,0") {
    angularDampingOverride=0.9999
    angularSpringConstant=75
    bAngularSpring=true
    bGravityOverrideInSimSpace=true
    bOverrideAngularDamping=true
    simulationSpace=RootRelative
    alphaBoolBlend {blendOption=Linear}
    boundBone {boneName=dyn_skirt_bk_03_r}
    componentLinearAccScale=<1.2 1.2 1.2>
    componentLinearVelScale=<1.2 1.2 1.2>
    physicsBodyDefinitions [{
      sphereCollisionRadius=0
      boundBone {boneName=dyn_skirt_bk_03_r}
      boxExtents=<20 20 10>
      constraintSetup {angularLimitsMax=<25 0 25>, angularLimitsMin=<-25 -25 -25>, angularTarget=<0 0 0>}
      localJointOffset=<2 0 0>
     }]
   }
   n30 = Component To Local(ComponentPose=@n29)
   n31 = RigidBody!(
    ComponentPose=Local To Component(LocalPose=Use cached pose 'PreDressRBAN'())
    Alpha
    OverrideWorldGravity:Vector=Get Gravity Override()
    ExternalForce
   ) -> (Pose) {
    bClampLinearTranslationLimitToRefPose=true
    bOverrideWorldGravity=true
    overridePhysicsAsset=&/F_MED_Punk_Koi_Physics
    simulationSpace=BaseBoneSpace
    alphaBoolBlend {blendOption=Linear}
    baseBoneRef {boneName=pelvis}
    componentLinearAccScale=<0.28 0.28 0.28>
    overrideWorldGravity {z=-980}
   }
   n32 = Component To Local(ComponentPose=@n31)
   n33 = "Constraint - dyn_skirt_fr_01_l_MSTR (PP)"!(
    ComponentPose=Local To Component(LocalPose=Blend Poses by bool(BlendPose_0=@n30, BlendPose_1=@n32, BlendTime_0=0.01, BlendTime_1=0.01) {blendType=Linear})
    Alpha=0.95
    ConstraintWeights_0=0
    ConstraintWeights_1:float=1
   ) -> (Pose) {
    alphaBoolBlend {blendOption=Linear}
    boneToModify {boneName=dyn_skirt_fr_01_l_MSTR}
    constraintSetup [{targetBone {boneName=pelvis}}, {targetBone {boneName=dyn_thigh_01_l}}]
    constraintWeights [1,1]
   }
   n34 = "Constraint - dyn_skirt_fr_01_l (PP)"!(ComponentPose=@n33, Alpha=0.95, ConstraintWeights_0=0, ConstraintWeights_1:float=1) -> (Pose) {
    alphaBoolBlend {blendOption=Linear}
    boneToModify {boneName=dyn_skirt_fr_01_l}
    constraintSetup [{targetBone {boneName=pelvis}}, {targetBone {boneName=dyn_thigh_01_l}}]
    constraintWeights [1,1]
   }
   n35 = "Constraint - dyn_skirt_fr_02_l (PP)"!(ComponentPose=@n34, Alpha=0.85, ConstraintWeights_0=0, ConstraintWeights_1:float=1) -> (Pose) {
    alphaBoolBlend {blendOption=Linear}
    boneToModify {boneName=dyn_skirt_fr_02_l}
    constraintSetup [{targetBone {boneName=pelvis}}, {targetBone {boneName=dyn_thigh_01_l}}]
    constraintWeights [1,1]
   }
   n36 = "Constraint - dyn_skirt_fr_03_l (PP)"!(ComponentPose=@n35, Alpha=0.6, ConstraintWeights_0=0, ConstraintWeights_1:float=1) -> (Pose) {
    alphaBoolBlend {blendOption=Linear}
    boneToModify {boneName=dyn_skirt_fr_03_l}
    constraintSetup [{targetBone {boneName=pelvis}}, {targetBone {boneName=dyn_thigh_01_l}}]
    constraintWeights [1,1]
   }
   n37 = "Constraint - dyn_skirt_fr2_01_l_MSTR (PP)"!(ComponentPose=@n36, Alpha=0.9, ConstraintWeights_0=0, ConstraintWeights_1:float=1) -> (Pose) {
    alphaBoolBlend {blendOption=Linear}
    boneToModify {boneName=dyn_skirt_fr2_01_l_MSTR}
    constraintSetup [{targetBone {boneName=pelvis}}, {targetBone {boneName=dyn_thigh_01_l}}]
    constraintWeights [1,1]
   }
   n38 = "Constraint - dyn_skirt_fr2_01_l (PP)"!(ComponentPose=@n37, Alpha=0.9, ConstraintWeights_0=0, ConstraintWeights_1:float=1) -> (Pose) {
    alphaBoolBlend {blendOption=Linear}
    boneToModify {boneName=dyn_skirt_fr2_01_l}
    constraintSetup [{targetBone {boneName=pelvis}}, {targetBone {boneName=dyn_thigh_01_l}}]
    constraintWeights [1,1]
   }
   n39 = "Constraint - dyn_skirt_fr2_02_l (PP)"!(ComponentPose=@n38, Alpha=0.8, ConstraintWeights_0=0, ConstraintWeights_1:float=1) -> (Pose) {
    alphaBoolBlend {blendOption=Linear}
    boneToModify {boneName=dyn_skirt_fr2_02_l}
    constraintSetup [{targetBone {boneName=pelvis}}, {targetBone {boneName=dyn_thigh_01_l}}]
    constraintWeights [1,1]
   }
   n40 = "Constraint - dyn_skirt_fr2_03_l (PP)"!(ComponentPose=@n39, Alpha=0.6, ConstraintWeights_0=0, ConstraintWeights_1:float=1) -> (Pose) {
    alphaBoolBlend {blendOption=Linear}
    boneToModify {boneName=dyn_skirt_fr2_03_l}
    constraintSetup [{targetBone {boneName=pelvis}}, {targetBone {boneName=dyn_thigh_01_l}}]
    constraintWeights [1,1]
   }
   n41 = "Constraint - dyn_skirt_fr_01_r_MSTR (PP)"!(ComponentPose=@n40, Alpha=0.9, ConstraintWeights_0=0, ConstraintWeights_1:float=1) -> (Pose) {
    alphaBoolBlend {blendOption=Linear}
    boneToModify {boneName=dyn_skirt_fr_01_r_MSTR}
    constraintSetup [{targetBone {boneName=pelvis}}, {targetBone {boneName=dyn_thigh_01_r}}]
    constraintWeights [1,1]
   }
   n42 = "Constraint - dyn_skirt_fr_01_r (PP)"!(ComponentPose=@n41, Alpha=0.9, ConstraintWeights_0=0, ConstraintWeights_1:float=1) -> (Pose) {
    alphaBoolBlend {blendOption=Linear}
    boneToModify {boneName=dyn_skirt_fr_01_r}
    constraintSetup [{targetBone {boneName=pelvis}}, {targetBone {boneName=dyn_thigh_01_r}}]
    constraintWeights [1,1]
   }
   n43 = "Constraint - dyn_skirt_fr_02_r (PP)"!(ComponentPose=@n42, Alpha=0.5, ConstraintWeights_0=0, ConstraintWeights_1:float=1) -> (Pose) {
    alphaBoolBlend {blendOption=Linear}
    boneToModify {boneName=dyn_skirt_fr_02_r}
    constraintSetup [{targetBone {boneName=pelvis}}, {targetBone {boneName=dyn_thigh_01_r}}]
    constraintWeights [1,1]
   }
   n44 = "Constraint - dyn_skirt_fr_03_r (PP)"!(ComponentPose=@n43, Alpha=0.5, ConstraintWeights_0=0, ConstraintWeights_1:float=1) -> (Pose) {
    alphaBoolBlend {blendOption=Linear}
    boneToModify {boneName=dyn_skirt_fr_03_r}
    constraintSetup [{targetBone {boneName=pelvis}}, {targetBone {boneName=dyn_thigh_01_r}}]
    constraintWeights [1,1]
   }
   n45 = "Constraint - dyn_skirt_fr2_01_r_MSTR (PP)"!(ComponentPose=@n44, Alpha=0.9, ConstraintWeights_0=0, ConstraintWeights_1:float=1) -> (Pose) {
    alphaBoolBlend {blendOption=Linear}
    boneToModify {boneName=dyn_skirt_fr2_01_r_MSTR}
    constraintSetup [{targetBone {boneName=pelvis}}, {targetBone {boneName=dyn_thigh_01_r}}]
    constraintWeights [1,1]
   }
   n46 = "Constraint - dyn_skirt_fr2_01_r (PP)"!(ComponentPose=@n45, Alpha=0.9, ConstraintWeights_0=0, ConstraintWeights_1:float=1) -> (Pose) {
    alphaBoolBlend {blendOption=Linear}
    boneToModify {boneName=dyn_skirt_fr2_01_r}
    constraintSetup [{targetBone {boneName=pelvis}}, {targetBone {boneName=dyn_thigh_01_r}}]
    constraintWeights [1,1]
   }
   n47 = "Constraint - dyn_skirt_fr2_02_r (PP)"!(ComponentPose=@n46, Alpha=0.5, ConstraintWeights_0=0, ConstraintWeights_1:float=1) -> (Pose) {
    alphaBoolBlend {blendOption=Linear}
    boneToModify {boneName=dyn_skirt_fr2_02_r}
    constraintSetup [{targetBone {boneName=pelvis}}, {targetBone {boneName=dyn_thigh_01_r}}]
    constraintWeights [1,1]
   }
   n48 = "Constraint - dyn_skirt_fr2_03_r (PP)"!(ComponentPose=@n47, Alpha=0.5, ConstraintWeights_0=0, ConstraintWeights_1:float=1) -> (Pose) {
    alphaBoolBlend {blendOption=Linear}
    boneToModify {boneName=dyn_skirt_fr2_03_r}
    constraintSetup [{targetBone {boneName=pelvis}}, {targetBone {boneName=dyn_thigh_01_r}}]
    constraintWeights [1,1]
   }
   n49 = "Constraint - dyn_skirt_fr_01_MSTR (PP)"!(ComponentPose=@n48, Alpha=1, ConstraintWeights_0=0.5, ConstraintWeights_1:float=0.5) -> (Pose) {
    alphaBoolBlend {blendOption=Linear}
    boneToModify {boneName=dyn_skirt_fr_01_MSTR}
    constraintSetup [{targetBone {boneName=dyn_skirt_fr_01_l_MSTR}}, {targetBone {boneName=dyn_skirt_fr_01_r_MSTR}}]
    constraintWeights [1,1]
   }
   n50 = "Constraint - dyn_skirt_fr_01 (PP)"!(ComponentPose=@n49, Alpha=1, ConstraintWeights_0=0.5, ConstraintWeights_1:float=0.5) -> (Pose) {
    alphaBoolBlend {blendOption=Linear}
    boneToModify {boneName=dyn_skirt_fr_01}
    constraintSetup [{targetBone {boneName=dyn_skirt_fr_01_l}}, {targetBone {boneName=dyn_skirt_fr_01_r}}]
    constraintWeights [1,1]
   }
   n51 = "Constraint - dyn_skirt_fr_02 (PP)"!(ComponentPose=@n50, Alpha=1, ConstraintWeights_0=0.5, ConstraintWeights_1:float=0.5) -> (Pose) {
    alphaBoolBlend {blendOption=Linear}
    boneToModify {boneName=dyn_skirt_fr_02}
    constraintSetup [{targetBone {boneName=dyn_skirt_fr_02_l}}, {targetBone {boneName=dyn_skirt_fr_02_r}}]
    constraintWeights [1,1]
   }
   n52 = "Constraint - dyn_skirt_fr_03 (PP)"!(ComponentPose=@n51, Alpha=1, ConstraintWeights_0=0.5, ConstraintWeights_1:float=0.5) -> (Pose) {
    alphaBoolBlend {blendOption=Linear}
    boneToModify {boneName=dyn_skirt_fr_03}
    constraintSetup [{targetBone {boneName=dyn_skirt_fr_03_l}}, {targetBone {boneName=dyn_skirt_fr_03_r}}]
    constraintWeights [1,1]
   }
   n53 = "Constraint - dyn_skirt_01_l_MSTR (P)"(ComponentPose=@n52, Alpha=0.5, ConstraintWeights_0=1) {
    alphaBoolBlend {blendOption=Linear}
    boneToModify {boneName=dyn_skirt_01_l_MSTR}
    constraintSetup [{targetBone {boneName=dyn_skirt_fr2_01_l_MSTR}}]
    constraintWeights [1]
   }
   n54 = "Constraint - dyn_skirt_01_l (P)"(ComponentPose=@n53, Alpha=0.5, ConstraintWeights_0=1) {
    alphaBoolBlend {blendOption=Linear}
    boneToModify {boneName=dyn_skirt_01_l}
    constraintSetup [{targetBone {boneName=dyn_skirt_fr2_01_l}}]
    constraintWeights [1]
   }
   n55 = "Constraint - dyn_skirt_02_l (P)"(ComponentPose=@n54, Alpha=0.5, ConstraintWeights_0=1) {
    alphaBoolBlend {blendOption=Linear}
    boneToModify {boneName=dyn_skirt_02_l}
    constraintSetup [{targetBone {boneName=dyn_skirt_fr2_02_l}}]
    constraintWeights [1]
   }
   n56 = "Constraint - dyn_skirt_03_l (P)"(ComponentPose=@n55, Alpha=0.5, ConstraintWeights_0=1) {
    alphaBoolBlend {blendOption=Linear}
    boneToModify {boneName=dyn_skirt_03_l}
    constraintSetup [{targetBone {boneName=dyn_skirt_fr2_03_l}}]
    constraintWeights [1]
   }
   n57 = "Constraint - dyn_skirt_bk2_01_l_MSTR (P)"(ComponentPose=@n56, Alpha=0.5, ConstraintWeights_0=1) {
    alphaBoolBlend {blendOption=Linear}
    boneToModify {boneName=dyn_skirt_bk2_01_l_MSTR}
    constraintSetup [{targetBone {boneName=dyn_skirt_01_l_MSTR}}]
    constraintWeights [1]
   }
   n58 = "Constraint - dyn_skirt_bk2_01_l (P)"(ComponentPose=@n57, Alpha=0.5, ConstraintWeights_0=1) {
    alphaBoolBlend {blendOption=Linear}
    boneToModify {boneName=dyn_skirt_bk2_01_l}
    constraintSetup [{targetBone {boneName=dyn_skirt_01_l}}]
    constraintWeights [1]
   }
   n59 = "Constraint - dyn_skirt_bk2_02_l (P)"(ComponentPose=@n58, Alpha=0.5, ConstraintWeights_0=1) {
    alphaBoolBlend {blendOption=Linear}
    boneToModify {boneName=dyn_skirt_bk2_02_l}
    constraintSetup [{targetBone {boneName=dyn_skirt_02_l}}]
    constraintWeights [1]
   }
   n60 = "Constraint - dyn_skirt_bk2_03_l (P)"(ComponentPose=@n59, Alpha=0.5, ConstraintWeights_0=1) {
    alphaBoolBlend {blendOption=Linear}
    boneToModify {boneName=dyn_skirt_bk2_03_l}
    constraintSetup [{targetBone {boneName=dyn_skirt_03_l}}]
    constraintWeights [1]
   }
   n61 = "Constraint - dyn_skirt_01_r_MSTR (P)"(ComponentPose=@n60, Alpha=0.5, ConstraintWeights_0=1) {
    alphaBoolBlend {blendOption=Linear}
    boneToModify {boneName=dyn_skirt_01_r_MSTR}
    constraintSetup [{targetBone {boneName=dyn_skirt_fr2_01_r_MSTR}}]
    constraintWeights [1]
   }
   n62 = "Constraint - dyn_skirt_01_r (P)"(ComponentPose=@n61, Alpha=0.5, ConstraintWeights_0=1) {
    alphaBoolBlend {blendOption=Linear}
    boneToModify {boneName=dyn_skirt_01_r}
    constraintSetup [{targetBone {boneName=dyn_skirt_fr2_01_r}}]
    constraintWeights [1]
   }
   n63 = "Constraint - dyn_skirt_02_r (P)"(ComponentPose=@n62, Alpha=0.5, ConstraintWeights_0=1) {
    alphaBoolBlend {blendOption=Linear}
    boneToModify {boneName=dyn_skirt_02_r}
    constraintSetup [{targetBone {boneName=dyn_skirt_fr2_02_r}}]
    constraintWeights [1]
   }
   n64 = "Constraint - dyn_skirt_03_r (P)"(ComponentPose=@n63, Alpha=0.5, ConstraintWeights_0=1) {
    alphaBoolBlend {blendOption=Linear}
    boneToModify {boneName=dyn_skirt_03_r}
    constraintSetup [{targetBone {boneName=dyn_skirt_fr2_03_r}}]
    constraintWeights [1]
   }
   n65 = "Constraint - dyn_skirt_bk2_01_r_MSTR (P)"(ComponentPose=@n64, Alpha=0.5, ConstraintWeights_0=1) {
    alphaBoolBlend {blendOption=Linear}
    boneToModify {boneName=dyn_skirt_bk2_01_r_MSTR}
    constraintSetup [{targetBone {boneName=dyn_skirt_01_r_MSTR}}]
    constraintWeights [1]
   }
   n66 = "Constraint - dyn_skirt_bk2_01_r (P)"(ComponentPose=@n65, Alpha=0.5, ConstraintWeights_0=1) {
    alphaBoolBlend {blendOption=Linear}
    boneToModify {boneName=dyn_skirt_bk2_01_r}
    constraintSetup [{targetBone {boneName=dyn_skirt_01_r}}]
    constraintWeights [1]
   }
   n67 = "Constraint - dyn_skirt_bk2_02_r (P)"(ComponentPose=@n66, Alpha=0.5, ConstraintWeights_0=1) {
    alphaBoolBlend {blendOption=Linear}
    boneToModify {boneName=dyn_skirt_bk2_02_r}
    constraintSetup [{targetBone {boneName=dyn_skirt_02_r}}]
    constraintWeights [1]
   }
   n68 = "Constraint - dyn_skirt_bk2_03_r (P)"(ComponentPose=@n67, Alpha=0.5, ConstraintWeights_0=1) {
    alphaBoolBlend {blendOption=Linear}
    boneToModify {boneName=dyn_skirt_bk2_03_r}
    constraintSetup [{targetBone {boneName=dyn_skirt_03_r}}]
    constraintWeights [1]
   }
   n69 = "Constraint - dyn_skirt_bk2_01_l_MSTR (PP)"!(ComponentPose=@n68, bAlphaBoolEnabled:bool=True, ConstraintWeights_0=0.5, ConstraintWeights_1:float=0.5) -> (Pose) {
    alphaInputType=Bool
    alphaBoolBlend {blendOption=Linear}
    boneToModify {boneName=dyn_skirt_bk2_01_l_MSTR}
    constraintSetup [{targetBone {boneName=dyn_skirt_bk_01_l_MSTR}}, {targetBone {boneName=dyn_skirt_01_l_MSTR}}]
    constraintWeights [1,1]
   }
   n70 = "Constraint - dyn_skirt_bk2_01_l (PP)"!(ComponentPose=@n69, bAlphaBoolEnabled:bool=True, ConstraintWeights_0=0.5, ConstraintWeights_1:float=0.5) -> (Pose) {
    alphaInputType=Bool
    alphaBoolBlend {blendOption=Linear}
    boneToModify {boneName=dyn_skirt_bk2_01_l}
    constraintSetup [{targetBone {boneName=dyn_skirt_bk_01_l}}, {targetBone {boneName=dyn_skirt_01_l}}]
    constraintWeights [1,1]
   }
   n71 = "Constraint - dyn_skirt_bk2_02_l (PP)"!(ComponentPose=@n70, bAlphaBoolEnabled:bool=True, ConstraintWeights_0=0.5, ConstraintWeights_1:float=0.5) -> (Pose) {
    alphaInputType=Bool
    alphaBoolBlend {blendOption=Linear}
    boneToModify {boneName=dyn_skirt_bk2_02_l}
    constraintSetup [{targetBone {boneName=dyn_skirt_bk_02_l}}, {targetBone {boneName=dyn_skirt_02_l}}]
    constraintWeights [1,1]
   }
   n72 = "Constraint - dyn_skirt_bk2_03_l (PP)"!(ComponentPose=@n71, bAlphaBoolEnabled:bool=True, ConstraintWeights_0=0.5, ConstraintWeights_1:float=0.5) -> (Pose) {
    alphaInputType=Bool
    alphaBoolBlend {blendOption=Linear}
    boneToModify {boneName=dyn_skirt_bk2_03_l}
    constraintSetup [{targetBone {boneName=dyn_skirt_bk_03_l}}, {targetBone {boneName=dyn_skirt_03_l}}]
    constraintWeights [1,1]
   }
   n73 = "Transform (Modify) Bone - Bone: dyn_skirt_bk2_01_l_MSTR"!(ComponentPose=@n72, Alpha=1, Translation="2,0,0", Rotation="0,0,10") -> (Pose) {
    rotationMode=BMM_Additive
    translationMode=BMM_Additive
    alphaBoolBlend {blendOption=Linear}
    boneToModify {boneName=dyn_skirt_bk2_01_l_MSTR}
    rotation {roll=10}
    translation {x=2}
   }
   n74 = "Constraint - dyn_skirt_01_l_MSTR (PP)"!(ComponentPose=@n73, bAlphaBoolEnabled:bool=True, ConstraintWeights_0=0.5, ConstraintWeights_1:float=0.5) -> (Pose) {
    alphaInputType=Bool
    alphaBoolBlend {blendOption=Linear}
    boneToModify {boneName=dyn_skirt_01_l_MSTR}
    constraintSetup [{targetBone {boneName=dyn_skirt_bk2_01_l_MSTR}}, {targetBone {boneName=dyn_skirt_fr2_01_l_MSTR}}]
    constraintWeights [1,1]
   }
   n75 = "Constraint - dyn_skirt_01_l (PP)"!(ComponentPose=@n74, bAlphaBoolEnabled:bool=True, ConstraintWeights_0=0.5, ConstraintWeights_1:float=0.5) -> (Pose) {
    alphaInputType=Bool
    alphaBoolBlend {blendOption=Linear}
    boneToModify {boneName=dyn_skirt_01_l}
    constraintSetup [{targetBone {boneName=dyn_skirt_bk2_01_l}}, {targetBone {boneName=dyn_skirt_fr2_01_l_MSTR}}]
    constraintWeights [1,1]
   }
   n76 = "Constraint - dyn_skirt_02_l (PP)"!(ComponentPose=@n75, bAlphaBoolEnabled:bool=True, ConstraintWeights_0=0.5, ConstraintWeights_1:float=0.5) -> (Pose) {
    alphaInputType=Bool
    alphaBoolBlend {blendOption=Linear}
    boneToModify {boneName=dyn_skirt_02_l}
    constraintSetup [{targetBone {boneName=dyn_skirt_bk2_02_l}}, {targetBone {boneName=dyn_skirt_fr2_02_l}}]
    constraintWeights [1,1]
   }
   n77 = "Constraint - dyn_skirt_03_l (PP)"!(ComponentPose=@n76, bAlphaBoolEnabled:bool=True, ConstraintWeights_0=0.5, ConstraintWeights_1:float=0.5) -> (Pose) {
    alphaInputType=Bool
    alphaBoolBlend {blendOption=Linear}
    boneToModify {boneName=dyn_skirt_03_l}
    constraintSetup [{targetBone {boneName=dyn_skirt_bk2_03_l}}, {targetBone {boneName=dyn_skirt_fr2_03_l}}]
    constraintWeights [1,1]
   }
   n78 = "Transform (Modify) Bone - Bone: dyn_skirt_01_l_MSTR"!(ComponentPose=@n77, bAlphaBoolEnabled:bool=True, Translation="0,0,3") -> (Pose) {
    alphaInputType=Bool
    rotationSpace=BCS_BoneSpace
    translationMode=BMM_Additive
    alphaBoolBlend {blendOption=Linear}
    boneToModify {boneName=dyn_skirt_01_l_MSTR}
    translation {z=3}
   }
   n79 = "Constraint - dyn_skirt_bk2_01_r_MSTR (PP)"!(ComponentPose=@n78, bAlphaBoolEnabled:bool=True, ConstraintWeights_0=0.5, ConstraintWeights_1:float=0.5) -> (Pose) {
    alphaInputType=Bool
    alphaBoolBlend {blendOption=Linear}
    boneToModify {boneName=dyn_skirt_bk2_01_r_MSTR}
    constraintSetup [{targetBone {boneName=dyn_skirt_bk_01_r_MSTR}}, {targetBone {boneName=dyn_skirt_01_r_MSTR}}]
    constraintWeights [1,1]
   }
   n80 = "Constraint - dyn_skirt_bk2_01_r (PP)"!(ComponentPose=@n79, bAlphaBoolEnabled:bool=True, ConstraintWeights_0=0.5, ConstraintWeights_1:float=0.5) -> (Pose) {
    alphaInputType=Bool
    alphaBoolBlend {blendOption=Linear}
    boneToModify {boneName=dyn_skirt_bk2_01_r}
    constraintSetup [{targetBone {boneName=dyn_skirt_bk_01_r}}, {targetBone {boneName=dyn_skirt_01_r}}]
    constraintWeights [1,1]
   }
   n81 = "Constraint - dyn_skirt_bk2_02_r (PP)"!(ComponentPose=@n80, bAlphaBoolEnabled:bool=True, ConstraintWeights_0=0.5, ConstraintWeights_1:float=0.5) -> (Pose) {
    alphaInputType=Bool
    alphaBoolBlend {blendOption=Linear}
    boneToModify {boneName=dyn_skirt_bk2_02_r}
    constraintSetup [{targetBone {boneName=dyn_skirt_bk_02_r}}, {targetBone {boneName=dyn_skirt_02_r}}]
    constraintWeights [1,1]
   }
   n82 = "Constraint - dyn_skirt_bk2_03_r (PP)"!(ComponentPose=@n81, bAlphaBoolEnabled:bool=True, ConstraintWeights_0=0.5, ConstraintWeights_1:float=0.5) -> (Pose) {
    alphaInputType=Bool
    alphaBoolBlend {blendOption=Linear}
    boneToModify {boneName=dyn_skirt_bk2_03_r}
    constraintSetup [{targetBone {boneName=dyn_skirt_bk_03_r}}, {targetBone {boneName=dyn_skirt_03_r}}]
    constraintWeights [1,1]
   }
   n83 = "Constraint - dyn_skirt_01_r_MSTR (PP)"!(ComponentPose=@n82, Alpha=1, ConstraintWeights_0=0.5, ConstraintWeights_1:float=0.5) -> (Pose) {
    alphaBoolBlend {blendOption=Linear}
    boneToModify {boneName=dyn_skirt_01_r_MSTR}
    constraintSetup [{targetBone {boneName=dyn_skirt_bk2_01_r_MSTR}}, {targetBone {boneName=dyn_skirt_fr2_01_r_MSTR}}]
    constraintWeights [1,1]
   }
   n84 = "Constraint - dyn_skirt_01_r (PP)"!(ComponentPose=@n83, Alpha=1, ConstraintWeights_0=0.5, ConstraintWeights_1:float=0.5) -> (Pose) {
    alphaBoolBlend {blendOption=Linear}
    boneToModify {boneName=dyn_skirt_01_r}
    constraintSetup [{targetBone {boneName=dyn_skirt_bk2_01_r}}, {targetBone {boneName=dyn_skirt_fr2_01_r}}]
    constraintWeights [1,1]
   }
   n85 = "Constraint - dyn_skirt_02_r (PP)"!(ComponentPose=@n84, Alpha=1, ConstraintWeights_0=0.5, ConstraintWeights_1:float=0.5) -> (Pose) {
    alphaBoolBlend {blendOption=Linear}
    boneToModify {boneName=dyn_skirt_02_r}
    constraintSetup [{targetBone {boneName=dyn_skirt_bk2_02_r}}, {targetBone {boneName=dyn_skirt_fr2_02_r}}]
    constraintWeights [1,1]
   }
   n86 = "Constraint - dyn_skirt_03_r (PP)"!(ComponentPose=@n85, Alpha=1, ConstraintWeights_0=0.5, ConstraintWeights_1:float=0.5) -> (Pose) {
    alphaBoolBlend {blendOption=Linear}
    boneToModify {boneName=dyn_skirt_03_r}
    constraintSetup [{targetBone {boneName=dyn_skirt_bk2_03_r}}, {targetBone {boneName=dyn_skirt_fr2_03_r}}]
    constraintWeights [1,1]
   }
   n87 = Anim Dynamics - Bone: dyn_skirt_fr_02_l(ComponentPose=@n86, Alpha=1, ExternalForce="0,0,0") {
    angularDampingOverride=0.9999
    angularSpringConstant=75
    bAngularSpring=true
    bGravityOverrideInSimSpace=true
    bOverrideAngularDamping=true
    simulationSpace=RootRelative
    alphaBoolBlend {blendOption=Linear}
    boundBone {boneName=dyn_skirt_fr_02_l}
    componentLinearAccScale=<1.2 1.2 1.2>
    componentLinearVelScale=<1.2 1.2 1.2>
    physicsBodyDefinitions [{
      sphereCollisionRadius=0
      boundBone {boneName=dyn_skirt_fr_02_l}
      boxExtents=<20 10 10>
      constraintSetup {angularLimitsMax=<25 0 25>, angularLimitsMin=<-25 -25 0>, angularTarget=<0 0 0>}
      localJointOffset=<2 0 0>
     }]
   }
   n88 = Anim Dynamics - Bone: dyn_skirt_fr_03_l(ComponentPose=@n87, Alpha=1, ExternalForce="0,0,0") {
    angularDampingOverride=0.9999
    angularSpringConstant=75
    bAngularSpring=true
    bGravityOverrideInSimSpace=true
    bOverrideAngularDamping=true
    simulationSpace=RootRelative
    alphaBoolBlend {blendOption=Linear}
    boundBone {boneName=dyn_skirt_fr_03_l}
    componentLinearAccScale=<1.2 1.2 1.2>
    componentLinearVelScale=<1.2 1.2 1.2>
    physicsBodyDefinitions [{
      sphereCollisionRadius=0
      boundBone {boneName=dyn_skirt_fr_03_l}
      boxExtents=<20 10 10>
      constraintSetup {angularLimitsMax=<25 0 25>, angularLimitsMin=<-25 -25 -25>, angularTarget=<0 0 0>}
      localJointOffset=<2 0 0>
     }]
   }
   n89 = Anim Dynamics - Bone: dyn_skirt_fr_02_r(ComponentPose=@n88, Alpha=1, ExternalForce="0,0,0") {
    angularDampingOverride=0.9999
    angularSpringConstant=75
    bAngularSpring=true
    bGravityOverrideInSimSpace=true
    bOverrideAngularDamping=true
    simulationSpace=RootRelative
    alphaBoolBlend {blendOption=Linear}
    boundBone {boneName=dyn_skirt_fr_02_r}
    componentLinearAccScale=<1.2 1.2 1.2>
    componentLinearVelScale=<1.2 1.2 1.2>
    physicsBodyDefinitions [{
      sphereCollisionRadius=0
      boundBone {boneName=dyn_skirt_fr_02_r}
      boxExtents=<20 20 10>
      constraintSetup {angularLimitsMax=<25 0 25>, angularLimitsMin=<-25 -25 -25>, angularTarget=<0 0 0>}
      localJointOffset=<2 0 0>
     }]
   }
   n90 = Anim Dynamics - Bone: dyn_skirt_fr_03_r(ComponentPose=@n89, Alpha=1, ExternalForce="0,0,0") {
    angularDampingOverride=0.9999
    angularSpringConstant=75
    bAngularSpring=true
    bGravityOverrideInSimSpace=true
    bOverrideAngularDamping=true
    simulationSpace=RootRelative
    alphaBoolBlend {blendOption=Linear}
    boundBone {boneName=dyn_skirt_fr_03_r}
    componentLinearAccScale=<1.2 1.2 1.2>
    componentLinearVelScale=<1.2 1.2 1.2>
    physicsBodyDefinitions [{
      sphereCollisionRadius=0
      boundBone {boneName=dyn_skirt_fr_03_r}
      boxExtents=<20 20 10>
      constraintSetup {angularLimitsMax=<25 0 25>, angularLimitsMin=<-25 -25 -25>, angularTarget=<0 0 0>}
      localJointOffset=<2 0 0>
     }]
   }
   n91 = Anim Dynamics - Bone: dyn_skirt_fr2_02_l(ComponentPose=@n90, Alpha=1, ExternalForce="0,0,0") {
    angularDampingOverride=0.9999
    angularSpringConstant=75
    bAngularSpring=true
    bGravityOverrideInSimSpace=true
    bOverrideAngularDamping=true
    simulationSpace=RootRelative
    alphaBoolBlend {blendOption=Linear}
    boundBone {boneName=dyn_skirt_fr2_02_l}
    componentLinearAccScale=<1.2 1.2 1.2>
    componentLinearVelScale=<1.2 1.2 1.2>
    physicsBodyDefinitions [{
      sphereCollisionRadius=0
      boundBone {boneName=dyn_skirt_fr2_02_l}
      boxExtents=<20 10 10>
      constraintSetup {angularLimitsMax=<25 0 25>, angularLimitsMin=<-25 -25 -25>, angularTarget=<0 0 0>}
      localJointOffset=<2 0 0>
     }]
   }
   n92 = Anim Dynamics - Bone: dyn_skirt_fr2_03_l(ComponentPose=@n91, Alpha=1, ExternalForce="0,0,0") {
    angularDampingOverride=0.9999
    angularSpringConstant=75
    bAngularSpring=true
    bGravityOverrideInSimSpace=true
    bOverrideAngularDamping=true
    simulationSpace=RootRelative
    alphaBoolBlend {blendOption=Linear}
    boundBone {boneName=dyn_skirt_fr2_03_l}
    componentLinearAccScale=<1.2 1.2 1.2>
    componentLinearVelScale=<1.2 1.2 1.2>
    physicsBodyDefinitions [{
      sphereCollisionRadius=0
      boundBone {boneName=dyn_skirt_fr2_03_l}
      boxExtents=<20 10 10>
      constraintSetup {angularLimitsMax=<25 0 25>, angularLimitsMin=<-25 -25 -25>, angularTarget=<0 0 0>}
      localJointOffset=<2 0 0>
     }]
   }
   n93 = Anim Dynamics - Bone: dyn_skirt_fr2_02_r(ComponentPose=@n92, Alpha=1, ExternalForce="0,0,0") {
    angularDampingOverride=0.9999
    angularSpringConstant=75
    bAngularSpring=true
    bGravityOverrideInSimSpace=true
    bOverrideAngularDamping=true
    simulationSpace=RootRelative
    alphaBoolBlend {blendOption=Linear}
    boundBone {boneName=dyn_skirt_fr2_02_r}
    componentLinearAccScale=<1.2 1.2 1.2>
    componentLinearVelScale=<1.2 1.2 1.2>
    physicsBodyDefinitions [{
      sphereCollisionRadius=0
      boundBone {boneName=dyn_skirt_fr2_02_r}
      boxExtents=<20 20 10>
      constraintSetup {angularLimitsMax=<25 0 25>, angularLimitsMin=<-25 -25 -25>, angularTarget=<0 0 0>}
      localJointOffset=<2 0 0>
     }]
   }
   n94 = Anim Dynamics - Bone: dyn_skirt_fr2_03_r(ComponentPose=@n93, Alpha=1, ExternalForce="0,0,0") {
    angularDampingOverride=0.9999
    angularSpringConstant=75
    bAngularSpring=true
    bGravityOverrideInSimSpace=true
    bOverrideAngularDamping=true
    simulationSpace=RootRelative
    alphaBoolBlend {blendOption=Linear}
    boundBone {boneName=dyn_skirt_fr2_03_r}
    componentLinearAccScale=<1.2 1.2 1.2>
    componentLinearVelScale=<1.2 1.2 1.2>
    physicsBodyDefinitions [{
      sphereCollisionRadius=0
      boundBone {boneName=dyn_skirt_fr2_03_r}
      boxExtents=<20 20 10>
      constraintSetup {angularLimitsMax=<25 0 25>, angularLimitsMin=<-25 -25 -25>, angularTarget=<0 0 0>}
      localJointOffset=<2 0 0>
     }]
   }
   n95 = Anim Dynamics - Bone: dyn_skirt_02_l(ComponentPose=@n94, Alpha=0.8, ExternalForce="0,0,0") {
    angularDampingOverride=0.9999
    angularSpringConstant=75
    bAngularSpring=true
    bGravityOverrideInSimSpace=true
    bOverrideAngularDamping=true
    simulationSpace=RootRelative
    alphaBoolBlend {blendOption=Linear}
    boundBone {boneName=dyn_skirt_02_l}
    componentLinearAccScale=<1.2 1.2 1.2>
    componentLinearVelScale=<1.2 1.2 1.2>
    physicsBodyDefinitions [{
      sphereCollisionRadius=0
      boundBone {boneName=dyn_skirt_02_l}
      boxExtents=<20 10 10>
      constraintSetup {angularLimitsMax=<25 0 25>, angularLimitsMin=<-25 -25 -25>, angularTarget=<0 0 0>}
      localJointOffset=<2 0 0>
     }]
   }
   n96 = Anim Dynamics - Bone: dyn_skirt_03_l(ComponentPose=@n95, Alpha=0.8, ExternalForce="0,0,0") {
    angularDampingOverride=0.9999
    angularSpringConstant=75
    bAngularSpring=true
    bGravityOverrideInSimSpace=true
    bOverrideAngularDamping=true
    simulationSpace=RootRelative
    alphaBoolBlend {blendOption=Linear}
    boundBone {boneName=dyn_skirt_03_l}
    componentLinearAccScale=<1.2 1.2 1.2>
    componentLinearVelScale=<1.2 1.2 1.2>
    physicsBodyDefinitions [{
      sphereCollisionRadius=0
      boundBone {boneName=dyn_skirt_03_l}
      boxExtents=<20 10 10>
      constraintSetup {angularLimitsMax=<25 0 25>, angularLimitsMin=<-25 -25 -25>, angularTarget=<0 0 0>}
      localJointOffset=<2 0 0>
     }]
   }
   n97 = Anim Dynamics - Bone: dyn_skirt_01_r(ComponentPose=@n96, Alpha=0.8, ExternalForce="0,0,0") {
    angularDampingOverride=0.9999
    angularSpringConstant=75
    bAngularSpring=true
    bGravityOverrideInSimSpace=true
    bOverrideAngularDamping=true
    simulationSpace=RootRelative
    alphaBoolBlend {blendOption=Linear}
    boundBone {boneName=dyn_skirt_01_r}
    componentLinearAccScale=<1.2 1.2 1.2>
    componentLinearVelScale=<1.2 1.2 1.2>
    physicsBodyDefinitions [{
      sphereCollisionRadius=0
      boundBone {boneName=dyn_skirt_01_r}
      boxExtents=<20 20 10>
      constraintSetup {angularLimitsMax=<25 0 25>, angularLimitsMin=<-25 -25 -25>, angularTarget=<0 0 0>}
      localJointOffset=<2 0 0>
     }]
   }
   n98 = Anim Dynamics - Bone: dyn_skirt_03_r(ComponentPose=@n97, Alpha=0.7, ExternalForce="0,0,0") {
    angularDampingOverride=0.9999
    angularSpringConstant=75
    bAngularSpring=true
    bGravityOverrideInSimSpace=true
    bOverrideAngularDamping=true
    simulationSpace=RootRelative
    alphaBoolBlend {blendOption=Linear}
    boundBone {boneName=dyn_skirt_03_r}
    componentLinearAccScale=<1.2 1.2 1.2>
    componentLinearVelScale=<1.2 1.2 1.2>
    physicsBodyDefinitions [{
      sphereCollisionRadius=0
      boundBone {boneName=dyn_skirt_03_r}
      boxExtents=<20 20 10>
      constraintSetup {angularLimitsMax=<25 0 25>, angularLimitsMin=<-25 -25 -25>, angularTarget=<0 0 0>}
      localJointOffset=<2 0 0>
     }]
   }
   n99 = Anim Dynamics - Bone: dyn_skirt_bk2_02_r(ComponentPose=@n98, Alpha=0.8, ExternalForce="0,0,0") {
    angularDampingOverride=0.9999
    angularSpringConstant=75
    bAngularSpring=true
    bGravityOverrideInSimSpace=true
    bOverrideAngularDamping=true
    simulationSpace=RootRelative
    alphaBoolBlend {blendOption=Linear}
    boundBone {boneName=dyn_skirt_bk2_02_r}
    componentLinearAccScale=<1.2 1.2 1.2>
    componentLinearVelScale=<1.2 1.2 1.2>
    physicsBodyDefinitions [{
      sphereCollisionRadius=0
      boundBone {boneName=dyn_skirt_bk2_02_r}
      boxExtents=<20 20 10>
      constraintSetup {angularLimitsMax=<25 0 25>, angularLimitsMin=<-25 -25 -25>, angularTarget=<0 0 0>}
      localJointOffset=<2 0 0>
     }]
   }
   n100 = Anim Dynamics - Bone: dyn_skirt_bk2_02_l(ComponentPose=@n99, Alpha=0.8, ExternalForce="0,0,0") {
    angularDampingOverride=0.9999
    angularSpringConstant=75
    bAngularSpring=true
    bGravityOverrideInSimSpace=true
    bOverrideAngularDamping=true
    simulationSpace=RootRelative
    alphaBoolBlend {blendOption=Linear}
    boundBone {boneName=dyn_skirt_bk2_02_l}
    componentLinearAccScale=<1.2 1.2 1.2>
    componentLinearVelScale=<1.2 1.2 1.2>
    physicsBodyDefinitions [{
      sphereCollisionRadius=0
      boundBone {boneName=dyn_skirt_bk2_02_l}
      boxExtents=<20 20 10>
      constraintSetup {angularLimitsMax=<25 0 25>, angularLimitsMin=<-25 -25 -25>, angularTarget=<0 0 0>}
      localJointOffset=<2 0 0>
     }]
   }
   n101 = Anim Dynamics - Bone: dyn_skirt_bk2_03_r(ComponentPose=@n100, Alpha=0.8, ExternalForce="0,0,0") {
    angularDampingOverride=0.9999
    angularSpringConstant=75
    bAngularSpring=true
    bGravityOverrideInSimSpace=true
    bOverrideAngularDamping=true
    simulationSpace=RootRelative
    alphaBoolBlend {blendOption=Linear}
    boundBone {boneName=dyn_skirt_bk2_03_r}
    componentLinearAccScale=<1.2 1.2 1.2>
    componentLinearVelScale=<1.2 1.2 1.2>
    physicsBodyDefinitions [{
      sphereCollisionRadius=0
      boundBone {boneName=dyn_skirt_bk2_03_r}
      boxExtents=<20 20 10>
      constraintSetup {angularLimitsMax=<25 0 25>, angularLimitsMin=<-25 -25 -25>, angularTarget=<0 0 0>}
      localJointOffset=<2 0 0>
     }]
   }
   n102 = Anim Dynamics - Bone: dyn_skirt_bk2_03_l(ComponentPose=@n101, Alpha=0.8, ExternalForce="0,0,0") {
    angularDampingOverride=0.9999
    angularSpringConstant=75
    bAngularSpring=true
    bGravityOverrideInSimSpace=true
    bOverrideAngularDamping=true
    simulationSpace=RootRelative
    alphaBoolBlend {blendOption=Linear}
    boundBone {boneName=dyn_skirt_bk2_03_l}
    componentLinearAccScale=<1.2 1.2 1.2>
    componentLinearVelScale=<1.2 1.2 1.2>
    physicsBodyDefinitions [{
      sphereCollisionRadius=0
      boundBone {boneName=dyn_skirt_bk2_03_l}
      boxExtents=<20 20 10>
      constraintSetup {angularLimitsMax=<25 0 25>, angularLimitsMin=<-25 -25 -25>, angularTarget=<0 0 0>}
      localJointOffset=<2 0 0>
     }]
   }
   n103 = "Transform (Modify) Bone - Bone: dyn_skirt_01_r_MSTR"!(ComponentPose=@n102, Alpha=1, Translation="0,0,2") -> (Pose) {
    rotationMode=BMM_Additive
    rotationSpace=BCS_BoneSpace
    scaleMode=BMM_Additive
    scaleSpace=BCS_BoneSpace
    translationMode=BMM_Additive
    translationSpace=BCS_BoneSpace
    alphaBoolBlend {blendOption=Linear}
    boneToModify {boneName=dyn_skirt_01_r_MSTR}
    translation {z=2}
   }
   n104 = Look At - Bone: dyn_shoeLip_l(ComponentPose=@n103, Alpha=1) {
    alphaBoolBlend {blendOption=Linear}
    boneToModify {boneName=dyn_shoeLip_l}
    lookAtLocation=<0 5>
    lookAtTarget {boneReference {boneName=calf_l}}
    lookAt_Axis {axis=<1 0>}
    lookUp_Axis {axis=<0 1>}
   }
   n105 = Look At - Bone: dyn_shoeLip_r(ComponentPose=@n104, Alpha=1) {
    alphaBoolBlend {blendOption=Linear}
    boneToModify {boneName=dyn_shoeLip_r}
    lookAtLocation=<0 -5>
    lookAtTarget {boneReference {boneName=calf_r}}
    lookAt_Axis {axis=<1 0>}
    lookUp_Axis {axis=<0 -1>}
   }
   n106 = "Constraint - dyn_skirt_fr_01_l (R)"!(ComponentPose=@n105, bAlphaBoolEnabled:bool=True, ConstraintWeights_0=1) -> (Pose) {
    alphaInputType=Bool
    alphaBoolBlend {blendOption=Linear}
    boneToModify {boneName=dyn_skirt_fr_01_l}
    constraintSetup [{transformType=Rotation, targetBone {boneName=thigh_twist_01_l}}]
    constraintWeights [1]
   }
   n107 = "Constraint - dyn_skirt_fr_02_l (R)"!(ComponentPose=@n106, bAlphaBoolEnabled:bool=True, ConstraintWeights_0=1) -> (Pose) {
    alphaInputType=Bool
    alphaBoolBlend {blendOption=Linear}
    boneToModify {boneName=dyn_skirt_fr_02_l}
    constraintSetup [{transformType=Rotation, targetBone {boneName=thigh_twist_01_l}}]
    constraintWeights [1]
   }
   n108 = "Constraint - dyn_skirt_fr_03_l (R)"!(ComponentPose=@n107, bAlphaBoolEnabled:bool=True, ConstraintWeights_0=1) -> (Pose) {
    alphaInputType=Bool
    alphaBoolBlend {blendOption=Linear}
    boneToModify {boneName=dyn_skirt_fr_03_l}
    constraintSetup [{transformType=Rotation, targetBone {boneName=thigh_twist_01_l}}]
    constraintWeights [1]
   }
   n109 = "Constraint - dyn_skirt_fr_01_r (R)"!(ComponentPose=@n108, bAlphaBoolEnabled:bool=True, ConstraintWeights_0=1) -> (Pose) {
    alphaInputType=Bool
    alphaBoolBlend {blendOption=Linear}
    boneToModify {boneName=dyn_skirt_fr_01_r}
    constraintSetup [{transformType=Rotation, targetBone {boneName=thigh_twist_01_r}}]
    constraintWeights [1]
   }
   n110 = "Constraint - dyn_skirt_fr_02_r (R)"!(ComponentPose=@n109, bAlphaBoolEnabled:bool=True, ConstraintWeights_0=1) -> (Pose) {
    alphaInputType=Bool
    alphaBoolBlend {blendOption=Linear}
    boneToModify {boneName=dyn_skirt_fr_02_r}
    constraintSetup [{transformType=Rotation, targetBone {boneName=thigh_twist_01_r}}]
    constraintWeights [1]
   }
   n111 = "Constraint - dyn_skirt_fr_03_r (R)"!(ComponentPose=@n110, bAlphaBoolEnabled:bool=True, ConstraintWeights_0=1) -> (Pose) {
    alphaInputType=Bool
    alphaBoolBlend {blendOption=Linear}
    boneToModify {boneName=dyn_skirt_fr_03_r}
    constraintSetup [{transformType=Rotation, targetBone {boneName=thigh_twist_01_r}}]
    constraintWeights [1]
   }
   n112 = "Constraint - dyn_skirt_fr_01 (R)"(ComponentPose=@n111, Alpha=1, ConstraintWeights_0=1) {
    alphaBoolBlend {blendOption=Linear}
    boneToModify {boneName=dyn_skirt_fr_01}
    constraintSetup [{transformType=Rotation, targetBone {boneName=thigh_twist_01_r}}]
    constraintWeights [1]
   }
   n113 = "Constraint - dyn_skirt_fr_02 (R)"(ComponentPose=@n112, Alpha=1, ConstraintWeights_0=1) {
    alphaBoolBlend {blendOption=Linear}
    boneToModify {boneName=dyn_skirt_fr_02}
    constraintSetup [{transformType=Rotation, targetBone {boneName=thigh_twist_01_r}}]
    constraintWeights [1]
   }
   n114 = "Constraint - dyn_skirt_fr_03 (R)"(ComponentPose=@n113, Alpha=1, ConstraintWeights_0=1) {
    alphaBoolBlend {blendOption=Linear}
    boneToModify {boneName=dyn_skirt_fr_03}
    constraintSetup [{transformType=Rotation, targetBone {boneName=thigh_twist_01_r}}]
    constraintWeights [1]
   }
   n115 = "Constraint - dyn_skirt_fr_01 (R)"(ComponentPose=@n114, Alpha=1, ConstraintWeights_0=1) {
    alphaBoolBlend {blendOption=Linear}
    boneToModify {boneName=dyn_skirt_fr_01}
    constraintSetup [{transformType=Rotation, targetBone {boneName=thigh_twist_01_l}}]
    constraintWeights [1]
   }
   n116 = "Constraint - dyn_skirt_fr_02 (R)"(ComponentPose=@n115, Alpha=1, ConstraintWeights_0=1) {
    alphaBoolBlend {blendOption=Linear}
    boneToModify {boneName=dyn_skirt_fr_02}
    constraintSetup [{transformType=Rotation, targetBone {boneName=thigh_twist_01_l}}]
    constraintWeights [1]
   }
   n117 = "Constraint - dyn_skirt_fr_03 (R)"(ComponentPose=@n116, Alpha=1, ConstraintWeights_0=1) {
    alphaBoolBlend {blendOption=Linear}
    boneToModify {boneName=dyn_skirt_fr_03}
    constraintSetup [{transformType=Rotation, targetBone {boneName=thigh_twist_01_l}}]
    constraintWeights [1]
   }
   n118 = "Transform (Modify) Bone - Bone: dyn_skirt_fr_01_r"(ComponentPose=@n117, Alpha=1, Translation="0,-2,1", Rotation="5,-10,0", Scale="1,1,1") {
    rotationMode=BMM_Additive
    rotationSpace=BCS_BoneSpace
    translationMode=BMM_Additive
    translationSpace=BCS_BoneSpace
    alphaBoolBlend {blendOption=Linear}
    boneToModify {boneName=dyn_skirt_fr_01_r}
    rotation {pitch=5, yaw=-10}
    translation {y=-2, z=1}
   }
   n119 = "Transform (Modify) Bone - Bone: dyn_skirt_fr_01_l"(ComponentPose=@n118, Alpha=1, Translation="0,2,1", Rotation="5,5,0", Scale="1,1,1") {
    rotationMode=BMM_Additive
    rotationSpace=BCS_BoneSpace
    translationMode=BMM_Additive
    translationSpace=BCS_BoneSpace
    alphaBoolBlend {blendOption=Linear}
    boneToModify {boneName=dyn_skirt_fr_01_l}
    rotation {pitch=5, yaw=5}
    translation {y=2, z=1}
   }
   n120 = "Transform (Modify) Bone - Bone: dyn_skirt_fr_01"!(ComponentPose=@n119, Alpha=1, Translation="0,0,2", Rotation="10,0,0") -> (Pose) {
    rotationMode=BMM_Additive
    rotationSpace=BCS_BoneSpace
    translationMode=BMM_Additive
    translationSpace=BCS_BoneSpace
    alphaBoolBlend {blendOption=Linear}
    boneToModify {boneName=dyn_skirt_fr_01}
    rotation {pitch=10}
    translation {z=2}
   }
   n121 = "Transform (Modify) Bone - Bone: dyn_skirt_fr_01"!(ComponentPose=@n120, Alpha=1, Translation="0,-2,0") -> (Pose) {
    rotationSpace=BCS_BoneSpace
    translationMode=BMM_Additive
    translationSpace=BCS_BoneSpace
    alphaBoolBlend {blendOption=Linear}
    boneToModify {boneName=dyn_skirt_fr_01}
    translation {y=-2}
   }
   n122 = "Transform (Modify) Bone - Bone: dyn_skirt_fr_01"(ComponentPose=@n121, Alpha=1, Translation="0,2,0") {
    rotationSpace=BCS_BoneSpace
    translationMode=BMM_Additive
    translationSpace=BCS_BoneSpace
    alphaBoolBlend {blendOption=Linear}
    boneToModify {boneName=dyn_skirt_fr_01}
    translation {y=2}
   }
   n123 = "Transform (Modify) Bone - Bone: dyn_skirt_bk_01_MSTR"!(ComponentPose=@n122, Alpha=1, Translation="0,0,2") -> (Pose) {
    rotationSpace=BCS_BoneSpace
    translationMode=BMM_Additive
    translationSpace=BCS_BoneSpace
    alphaBoolBlend {blendOption=Linear}
    boneToModify {boneName=dyn_skirt_bk_01_MSTR}
    translation {z=2}
   }
   n124 = "Transform (Modify) Bone - Bone: dyn_skirt_bk_01_r_MSTR"!(ComponentPose=@n123, Alpha=1, Translation="0,0,2") -> (Pose) {
    rotationSpace=BCS_BoneSpace
    translationMode=BMM_Additive
    translationSpace=BCS_BoneSpace
    alphaBoolBlend {blendOption=Linear}
    boneToModify {boneName=dyn_skirt_bk_01_r_MSTR}
    translation {z=2}
   }
   n125 = "Transform (Modify) Bone - Bone: dyn_skirt_bk_01_l_MSTR"!(ComponentPose=@n124, Alpha=1, Translation="0,0,2") -> (Pose) {
    rotationSpace=BCS_BoneSpace
    translationMode=BMM_Additive
    translationSpace=BCS_BoneSpace
    alphaBoolBlend {blendOption=Linear}
    boneToModify {boneName=dyn_skirt_bk_01_l_MSTR}
    translation {z=2}
   }
   Output Pose(Result=Component To Local(ComponentPose=@n125)) {layerGroup=None}
   n126 = "Constraint - dyn_thigh_01_l (RR)"!(
    ComponentPose=Local To Component(LocalPose=Copy Pose From Mesh() {bUseAttachedParent=true})
    Alpha=1
    ConstraintWeights_0=1
    ConstraintWeights_1:float=0
   ) -> (Pose) {
    alphaBoolBlend {blendOption=Linear}
    boneToModify {boneName=dyn_thigh_01_l}
    constraintSetup [{transformType=Rotation, targetBone {boneName=thigh_l}}, {transformType=Rotation, targetBone {boneName=pelvis}}]
    constraintWeights [1,1]
   }
   n127 = "Constraint - dyn_thigh_01_r (RR)"!(ComponentPose=@n126, Alpha=1, ConstraintWeights_0=1, ConstraintWeights_1:float=0) -> (Pose) {
    alphaBoolBlend {blendOption=Linear}
    boneToModify {boneName=dyn_thigh_01_r}
    constraintSetup [{transformType=Rotation, targetBone {boneName=thigh_r}}, {transformType=Rotation, targetBone {boneName=pelvis}}]
    constraintWeights [1,1]
   }
   n128 = "Constraint - dyn_skirt_fr_01_MSTR (RR)"!(ComponentPose=@n127, Alpha=1, ConstraintWeights_0=0.5, ConstraintWeights_1:float=0.5) -> (Pose) {
    alphaBoolBlend {blendOption=Linear}
    boneToModify {boneName=dyn_skirt_fr_01_MSTR}
    constraintSetup [
     {transformType=Rotation, targetBone {boneName=dyn_skirt_fr_01_l_MSTR}}
     {transformType=Rotation, targetBone {boneName=dyn_skirt_fr_01_r_MSTR}}
    ]
    constraintWeights [1,1]
   }
   n129 = "Constraint - dyn_skirt_fr_01 (RR)"!(ComponentPose=@n128, Alpha=1, ConstraintWeights_0=0.5, ConstraintWeights_1:float=0.5) -> (Pose) {
    alphaBoolBlend {blendOption=Linear}
    boneToModify {boneName=dyn_skirt_fr_01}
    constraintSetup [{transformType=Rotation, targetBone {boneName=dyn_skirt_fr_01_l}}, {transformType=Rotation, targetBone {boneName=dyn_skirt_fr_01_r}}]
    constraintWeights [1,1]
   }
   n130 = "Constraint - dyn_skirt_fr_02 (RR)"!(ComponentPose=@n129, Alpha=1, ConstraintWeights_0=0.5, ConstraintWeights_1:float=0.5) -> (Pose) {
    alphaBoolBlend {blendOption=Linear}
    boneToModify {boneName=dyn_skirt_fr_02}
    constraintSetup [{transformType=Rotation, targetBone {boneName=dyn_skirt_fr_02_l}}, {transformType=Rotation, targetBone {boneName=dyn_skirt_fr_02_r}}]
    constraintWeights [1,1]
   }
   n131 = "Constraint - dyn_skirt_fr_03 (RR)"!(ComponentPose=@n130, Alpha=1, ConstraintWeights_0=0.5, ConstraintWeights_1:float=0.5) -> (Pose) {
    alphaBoolBlend {blendOption=Linear}
    boneToModify {boneName=dyn_skirt_fr_03}
    constraintSetup [{transformType=Rotation, targetBone {boneName=dyn_skirt_fr_03_l}}, {transformType=Rotation, targetBone {boneName=dyn_skirt_fr_03_r}}]
    constraintWeights [1,1]
   }
   Save cached pose 'PreDressRBAN'(Pose=Component To Local(ComponentPose=@n131)) {cacheName=PreDressRBAN}
  ]}
 EventGraph {nodes [
   Event BlueprintUpdateAnimation()
   > n1 = Sequence!(execute) -> (then_0, then_1, then_2:Exec, then_3:Exec, then_4:Exec, then_5:Exec)
   > n2 = Gravity Override State Function DEPRECATED!(
    execute
    CharacterPartAnimInstance=Self-Reference()
    GameTimeInSeconds=GetGameTimeInSeconds()
    GravityOverrideParams_Skydive_GoingUp_28_#1_WindFrequency_28_#2:Vector="1.5,7,8"
    GravityOverrideParams_Skydive_GoingUp_28_#1_WindAmplitude_31_#3:Vector="200,-500,400"
    GravityOverrideParams_Skydive_GoingUp_28_#1_WindOffset_32_#4:Vector="0,0,0"
    GravityOverrideParams_Skydive_GoingUp_28_#1_GravityOverride_15_#5:Vector="0,0,-100"
    GravityOverrideParams_Skydive_GoingUp_28_#1_JointName_27_#6:Name=root
    GravityOverrideParams_Skydive_FloatingDown_29_#7_WindFrequency_28_#2:Vector="1.5,7,8"
    GravityOverrideParams_Skydive_FloatingDown_29_#7_WindAmplitude_31_#3:Vector="200,-500,400"
    GravityOverrideParams_Skydive_FloatingDown_29_#7_WindOffset_32_#4:Vector="0,-600,400"
    GravityOverrideParams_Skydive_FloatingDown_29_#7_GravityOverride_15_#5:Vector="0,0,0"
    GravityOverrideParams_Skydive_FloatingDown_29_#7_JointName_27_#6:Name=root
    GravityOverrideParams_Skydive_DivingDown_30_#8_WindFrequency_28_#2:Vector="1.5,7,8"
    GravityOverrideParams_Skydive_DivingDown_30_#8_WindAmplitude_31_#3:Vector="50,300,100"
    GravityOverrideParams_Skydive_DivingDown_30_#8_WindOffset_32_#4:Vector="0,0,0"
    GravityOverrideParams_Skydive_DivingDown_30_#8_GravityOverride_15_#5:Vector="0,0,0"
    GravityOverrideParams_Skydive_DivingDown_30_#8_JointName_27_#6:Name=pelvis
    GravityOverrideParams_Skydive_Parachuting_31_#9_WindFrequency_28_#2:Vector="1.5,7,8"
    GravityOverrideParams_Skydive_Parachuting_31_#9_WindAmplitude_31_#3:Vector="200,-500,400"
    GravityOverrideParams_Skydive_Parachuting_31_#9_WindOffset_32_#4:Vector="0,0,0"
    GravityOverrideParams_Skydive_Parachuting_31_#9_GravityOverride_15_#5:Vector="0,-800,-980"
    GravityOverrideParams_Skydive_Parachuting_31_#9_JointName_27_#6:Name=root
    GravityOverrideParams_OnGround_Standing_32_#10:Gravity Override Setting Struct Structure=(WindFrequency_28_#2=(X=0,Y=0,Z=0),WindAmplitude_31_#3=(X=0,Y=0,Z=0),WindOffset_32_#4=(X=0,Y=0,Z=0),GravityOverride_15_#5=(X=0,Y=0,Z=-980),JointName_27_#6="root")
    GravityOverrideParams_OnGround_Moving_33_#11:Gravity Override Setting Struct Structure=(WindFrequency_28_#2=(X=0,Y=0,Z=0),WindAmplitude_31_#3=(X=0,Y=0,Z=0),WindOffset_32_#4=(X=0,Y=0,Z=0),GravityOverride_15_#5=(X=0,Y=0,Z=-980),JointName_27_#6="root")
    GravityOverrideParams_EmoteOrMelee_34_#12:Gravity Override Setting Struct Structure=(WindFrequency_28_#2=(X=0,Y=0,Z=0),WindAmplitude_31_#3=(X=0,Y=0,Z=0),WindOffset_32_#4=(X=0,Y=0,Z=0),GravityOverride_15_#5=(X=0,Y=0,Z=-980),JointName_27_#6="root")
    GravityOverrideParams_Swimming_35_#13:Gravity Override Setting Struct Structure=(WindFrequency_28_#2=(X=0,Y=0,Z=0),WindAmplitude_31_#3=(X=0,Y=0,Z=0),WindOffset_32_#4=(X=0,Y=0,Z=0),GravityOverride_15_#5=(X=-800,Y=0,Z=0),JointName_27_#6="pelvis")
    GravityOverrideParams_Front_End_36_#14:Gravity Override Setting Struct Structure=(WindFrequency_28_#2=(X=0,Y=0,Z=0),WindAmplitude_31_#3=(X=0,Y=0,Z=0),WindOffset_32_#4=(X=0,Y=0,Z=0),GravityOverride_15_#5=(X=0,Y=0,Z=-980),JointName_27_#6="root")
    GravityOverrideParams_DBNO_40_#15:Gravity Override Setting Struct Structure=(WindFrequency_28_#2=(X=0,Y=0,Z=0),WindAmplitude_31_#3=(X=0,Y=0,Z=0),WindOffset_32_#4=(X=0,Y=0,Z=0),GravityOverride_15_#5=(X=-980,Y=400,Z=0),JointName_27_#6="pelvis")
    GravityOverrideParams_DBNO_Carried_41_#16:Gravity Override Setting Struct Structure=(WindFrequency_28_#2=(X=0,Y=0,Z=0),WindAmplitude_31_#3=(X=0,Y=0,Z=0),WindOffset_32_#4=(X=0,Y=0,Z=0),GravityOverride_15_#5=(X=-980,Y=500,Z=-500),JointName_27_#6="pelvis")
    GravityOverrideParams_OnGround_Falling_45_#17:Gravity Override Setting Struct Structure=(WindFrequency_28_#2=(X=0,Y=0,Z=0),WindAmplitude_31_#3=(X=0,Y=0,Z=0),WindOffset_32_#4=(X=0,Y=0,Z=0),GravityOverride_15_#5=(X=0,Y=0,Z=-980),JointName_27_#6="root")
    GravityOverrideParams_Ziplining_46_#18:Gravity Override Setting Struct Structure=(WindFrequency_28_#2=(X=0,Y=0,Z=0),WindAmplitude_31_#3=(X=0,Y=0,Z=0),WindOffset_32_#4=(X=0,Y=0,Z=0),GravityOverride_15_#5=(X=0,Y=0,Z=-980),JointName_27_#6="root")
    GravityOverrideParams_Skydive_Surfing_49_#19:Gravity Override Setting Struct Structure=(WindFrequency_28_#2=(X=1.5,Y=7,Z=8),WindAmplitude_31_#3=(X=200,Y=-500,Z=400),WindOffset_32_#4=(X=0,Y=0,Z=0),GravityOverride_15_#5=(X=500,Y=-300,Z=-980),JointName_27_#6="root")
    GravityOverrideParams_Riding_Standing_54_#20:Gravity Override Setting Struct Structure=(WindFrequency_28_#2=(X=0,Y=0,Z=0),WindAmplitude_31_#3=(X=0,Y=0,Z=0),WindOffset_32_#4=(X=0,Y=0,Z=0),GravityOverride_15_#5=(X=0,Y=-350,Z=-980),JointName_27_#6="root")
    GravityOverrideParams_Riding_Moving_55_#21:Gravity Override Setting Struct Structure=(WindFrequency_28_#2=(X=0,Y=0,Z=0),WindAmplitude_31_#3=(X=0,Y=0,Z=0),WindOffset_32_#4=(X=0,Y=0,Z=0),GravityOverride_15_#5=(X=0,Y=-450,Z=-450),JointName_27_#6="root")
    GravityOverrideParams_Riding_Moving_Targeting_58_#22:Gravity Override Setting Struct Structure=(WindFrequency_28_#2=(X=0,Y=0,Z=0),WindAmplitude_31_#3=(X=0,Y=0,Z=0),WindOffset_32_#4=(X=0,Y=0,Z=0),GravityOverride_15_#5=(X=0,Y=-200,Z=-980),JointName_27_#6="root")
    GravityOverrideParams_Tactical_Sprint_57_#23:Gravity Override Setting Struct Structure=(WindFrequency_28_#2=(X=0,Y=0,Z=0),WindAmplitude_31_#3=(X=0,Y=0,Z=0),WindOffset_32_#4=(X=0,Y=0,Z=0),GravityOverride_15_#5=(X=0,Y=-450,Z=-500),JointName_27_#6="root")
    GravityOverrideParams_Rail_Grinding_61_#24:Gravity Override Setting Struct Structure=(WindFrequency_28_#2=(X=0,Y=0,Z=0),WindAmplitude_31_#3=(X=0,Y=0,Z=0),WindOffset_32_#4=(X=0,Y=0,Z=0),GravityOverride_15_#5=(X=0,Y=0,Z=-980),JointName_27_#6="root")
    GravityOverrideParams_DBNO_Tumbling_67_#25:Gravity Override Setting Struct Structure=(WindFrequency_28_#2=(X=0,Y=0,Z=0),WindAmplitude_31_#3=(X=0,Y=0,Z=0),WindOffset_32_#4=(X=0,Y=0,Z=0),GravityOverride_15_#5=(X=-2000,Y=0,Z=0),JointName_27_#6="pelvis")
    GravityOverrideParams_DBNO_RollFwd_68_#26:Gravity Override Setting Struct Structure=(WindFrequency_28_#2=(X=0,Y=0,Z=0),WindAmplitude_31_#3=(X=0,Y=0,Z=0),WindOffset_32_#4=(X=0,Y=0,Z=0),GravityOverride_15_#5=(X=-2000,Y=0,Z=0),JointName_27_#6="pelvis")
    GravityOverrideParams_DBNO_RollBwd_69_#27:Gravity Override Setting Struct Structure=(WindFrequency_28_#2=(X=0,Y=0,Z=0),WindAmplitude_31_#3=(X=0,Y=0,Z=0),WindOffset_32_#4=(X=0,Y=0,Z=0),GravityOverride_15_#5=(X=-4000,Y=0,Z=0),JointName_27_#6="pelvis")
    GravityOverrideParams_DBNO_Impulse_70_#28:Gravity Override Setting Struct Structure=(WindFrequency_28_#2=(X=0,Y=0,Z=0),WindAmplitude_31_#3=(X=0,Y=0,Z=0),WindOffset_32_#4=(X=0,Y=0,Z=0),GravityOverride_15_#5=(X=-2000,Y=0,Z=0),JointName_27_#6="pelvis")
    RBAN Params
    DebugDraw
    DebugJointName
   ) -> (then, GravityOverride, CurrentStateOut, GravityOverrideSettingsOut, RBAN SettingsOut)
   > Set Gravity Override(Gravity Override=@n2)
   n3 = GetPartSkeletalMeshComponent()
   > Is Valid(InputObject=@n3)
   > Delay(Duration=0)
   > SetVectorParameterValueOnMaterials!(
    execute
    self:Mesh Component Object Reference=@n3
    ParameterName:Name=Cube Location
    ParameterValue:Vector=TransformLocation!(T:"Transform (by ref)"=GetSocketTransform(self=@n3, InSocketName=spine_05, TransformSpace=RTS_Component), Location:Vector="0, 0, 0") -> (ReturnValue:Vector)
   ) -> (then)
   > SetVectorParameterValueOnMaterials(
    self=@n3
    ParameterName=Core Location R
    ParameterValue=TransformLocation(T=GetSocketTransform(self=@n3, InSocketName=foot_r, TransformSpace=RTS_Component), Location="-9.5,-1.5,-0.25")
   )
   > SetVectorParameterValueOnMaterials(
    self=@n3
    ParameterName=Core Location L
    ParameterValue=TransformLocation(T=GetSocketTransform(self=@n3, InSocketName=foot_l, TransformSpace=RTS_Component), Location="9.5,1.5,0.25")
   )
   Set Skydiving(execute=@n1.then_2, Skydiving:bool=OR Boolean(A=Get bIsSkyDiving(), B=Get bIsSkydiveDiveMode())) -> (Output_Get:bool)
   n4 = Get R Thigh rl Rot() -> (R Thigh rl Rot:double)
   n5 = Get L Thigh rl Rot() -> (L Thigh rl Rot:double)
   Set Thigh_right(execute=@n1.then_3, Thigh_right:double=SelectFloat(A=@n4, B=0, bPickA="float >= float"(A=@n4, B=@n5) {is=title K2Node_CallFunction})) -> (Output_Get:double)
   n6 = Get R Thigh rl Rot()
   n7 = Get L Thigh rl Rot()
   Set Thigh_left(execute=@n1.then_4, Thigh_left:double=SelectFloat(A=@n7, B=0, bPickA="float >= float"(A=@n7, B=@n6))) -> (Output_Get:double)
   n8 = Get R Thigh fb Rot() -> (R Thigh fb Rot:double)
   n9 = Get L Thigh fb Rot() -> (L Thigh fb Rot:double)
   Set Thighs(execute=@n1.then_5, Thighs:double=SelectFloat(A=@n8, B=@n9, bPickA="float >= float"(A=@n8, B=@n9))) -> (Output_Get:double)
   n10 = Branch(execute=@n1.then_1, Condition=Get bIsPlayingEmote())
   n11 = GetOwnerSkeletalMeshComponent()
   n12 = BreakRotator(InRot=BreakTransform(InTransform=GetDeltaTransformFromRefPose(self=@n11, BoneName=thigh_r, BaseName=pelvis)).Rotation)
   > Set R Thigh fb Rot(R Thigh fb Rot:double=MapRangeClamped(Value=@n12.Yaw, InRangeA=-10, InRangeB=-75, OutRangeA=0.3)) -> (Output_Get:double)
   > Set R Thigh rl Rot(R Thigh rl Rot:double=MapRangeClamped(Value=@n12, InRangeA=0, InRangeB=45)) -> (Output_Get:double)
   n13 = GetOwnerSkeletalMeshComponent()
   n14 = BreakRotator(InRot=BreakTransform(InTransform=GetDeltaTransformFromRefPose(self=@n13, BoneName=thigh_l, BaseName=pelvis)).Rotation)
   > Set L Thigh fb Rot(L Thigh fb Rot:double=MapRangeClamped(Value=@n14.Yaw, InRangeA=-10, InRangeB=-75, OutRangeA=0.3)) -> (Output_Get:double)
   > Set L Thigh rl Rot(L Thigh rl Rot:double=MapRangeClamped(Value=@n14, InRangeA=0, InRangeB=45)) -> (Output_Get:double)
   > Set Back Skirt Emote Alpha(Back Skirt Emote Alpha:double=1) -> (Output_Get:double)
   Set R Thigh fb Rot(execute=@n10.else, R Thigh fb Rot=0)
   > Set L Thigh fb Rot(L Thigh fb Rot=0)
   > Set R Thigh rl Rot(R Thigh rl Rot=0)
   > Set L Thigh rl Rot(L Thigh rl Rot=0)
   > Set Back Skirt Emote Alpha(Back Skirt Emote Alpha=0)
  ]}
]
variables [
 Back Skirt Emote Alpha {type=number}
 Gravity Override {type=Vector}
 L Thigh fb Rot {type=number}
 L Thigh rl Rot {type=number}
 R Thigh fb Rot {type=number}
 R Thigh rl Rot {type=number}
 Skydiving {type=boolean}
 Thigh_left {type=number}
 Thigh_right {type=number}
 Thighs {type=number}
]
